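# Battle Royale HUD widget for contextual controls that displays a keybind, optionally followed by a "/" separator.
# Derives from FortKeybindWidgetHUDFeedback, with a horizontal box holding the inherited Keybind_Action widget, set to show unbound status, and a KeybindSeparator text block.
# The ShowDelineator event switches the separator between HitTestInvisible and Collapsed according to its bShouldShow flag.
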WidgetBlueprint /BattleRoyaleGameUI/HUD/ContextualControls/WBP_BRGameUI_ContextualControls_KeybindWidget
graphs [EventGraph {nodes [
   n1 = Event ShowDelineator!() -> (OutputDelegate, then, bShouldShow:bool) {is=title K2Node_Event}
   > n2 = Branch(Condition=@n1.bShouldShow)
   n3 = Get KeybindSeparator() -> (KeybindSeparator:Text Block Object Reference)
   > SetVisibility(self=@n3, InVisibility=HitTestInvisible)
   SetVisibility(execute=@n2.else, self=@n3, InVisibility=Collapsed)
  ]}]
variables []
widgets {
 info {
  inheritedWidgetCount=1
  parentClass=/Script/FortniteUI.FortKeybindWidgetHUDFeedback
  rootWidgetClass=&HorizontalBox
  widgetCount=3
 }
 widgets [
  HorizontalBox_30 {widgetClassPath=&HorizontalBox, widget=HorizontalBox_30}
  Keybind_Action {
   widgetClassPath=/BattleRoyaleGameUI/HUD/KeybindWidget/WBP_BRGameUI_KeybindWidget
   bInherited=true
   parent=HorizontalBox_30
   widget=Keybind_Action
   properties {bShowUnboundStatus=true, brushSize=Medium}
   slot {class=HorizontalBoxSlot, properties {horizontalAlignment=HAlign_Center, verticalAlignment=VAlign_Center}}
  }
  KeybindSeparator {
   widgetClassPath=&UEFN_TextBlock
   bIsVariable=true
   parent=HorizontalBox_30
   widget=KeybindSeparator
   properties {
    text=/
    font {
     bForceMonospaced=false
     bMaterialIsStencil=false
     fontMaterial=/BattleRoyaleGameUI/Shared/M_UI_Font_White
     fontObject=&BurbankSmall
     letterSpacing=0
     monospacedWidth=1
     size=18
     skewAmount=0
     typefaceFontName=Black
     outlineSettings {
      bApplyOutlineToDropShadows=true
      bMiteredCorners=false
      bSeparateFillAlpha=true
      outlineMaterial=/BattleRoyaleGameUI/Shared/MI_UI_Font_SoftShadow_SmallText
      outlineSize=4
      outlineColor=<rgba 1 1 1 1>
     }
    }
    margin=<ltrb -2 0 -2 0>
   }
   slot {class=HorizontalBoxSlot, properties {horizontalAlignment=HAlign_Left, verticalAlignment=VAlign_Center, padding=<ltrb 5 0 0 0>}}
  }
 ]
}
