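# Single contextual-controls HUD line, built on AthenaHUDFeedbackLineCooldown, showing keybind entries, uppercase action text and additional rich-text feedback.
# Enabled state tints the content border, and cooldown start and finish dim and restore the whole widget.
# Event_DangerPulse toggles a red-yellow glow and plays or stops its pulse animation at a given speed.
# IsKeyBound searches visible keybind entries for a key, and SetTextStyleSet applies one style table to all three text blocks.
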WidgetBlueprint /BattleRoyaleGameUI/HUD/ContextualControls/WBP_BRGameUI_ContextualControls_Line
animations [DangerPulse {
  binding_count=3
  length_seconds=0.8521
  object=&self:DangerPulse
  sequence {bindings [
    Image_YellowBG {
     id=#1
     possessed_class=Image
     tracks [
      {
       type=MovieSceneFloatTrack
       display_name=Render Opacity
       sections [{type=MovieSceneFloatSection, channels [None {type=MovieSceneScriptingFloatChannel, keys [[2.9167,1],[19.9583,0.2]]}]}]
      }
      {
       type=MovieScene2DTransformTrack
       sections [{
         type=MovieScene2DTransformSection
         channels [
          Scale.X {type=MovieSceneScriptingFloatChannel, keys [[2.9167,1],[19.9583,0.8]]}
          Scale.Y {type=MovieSceneScriptingFloatChannel, keys [[2.9167,1],[19.9583,0.8]]}
         ]
        }]
      }
     ]
    }
    Image_RedBG {
     id=#2
     possessed_class=Image
     tracks [
      {
       type=MovieSceneFloatTrack
       display_name=Render Opacity
       sections [{type=MovieSceneFloatSection, channels [None {type=MovieSceneScriptingFloatChannel}]}]
      }
      {
       type=MovieScene2DTransformTrack
       sections [{
         type=MovieScene2DTransformSection
         channels [
          Scale.X {type=MovieSceneScriptingFloatChannel, keys [[2.9167,1],[19.9583,0.8]]}
          Scale.Y {type=MovieSceneScriptingFloatChannel, keys [[2.9167,1],[19.9583,0.8]]}
         ]
        }]
      }
     ]
    }
    Overlay_Glow {
     id=#3
     possessed_class=Overlay
     tracks [{
       type=MovieSceneFloatTrack
       display_name=Render Opacity
       sections [{type=MovieSceneFloatSection, channels [None {type=MovieSceneScriptingFloatChannel, keys [[2.9167,1],[19.9583,0.4]]}]}]
      }]
    }
   ]}
 }]
graphs [
 EventGraph {nodes [
   n1 = Event_DangerPulse!() -> (OutputDelegate, then, Enable:bool=true, AnimSpeed:double=~) {is=title K2Node_CustomEvent}
   > SetVisibility(
    self=Get Overlay_Glow() -> (Overlay_Glow:Overlay Object Reference)
    InVisibility=Select(Option 0=Hidden, Option 1=HitTestInvisible, Index=@n1.Enable)
   )
   > n2 = Branch(Condition=@n1.Enable)
   n3 = Get DangerPulse() -> (DangerPulse:Widget Animation Object Reference)
   > PlayAnimation(InAnimation=@n3, PlaybackSpeed=@n1.AnimSpeed)
   StopAnimation(execute=@n2.else, InAnimation=@n3)
   n4 = Event OnShowEnabledUpdated!() -> (OutputDelegate, then, bShowEnabled:bool) {is=title K2Node_Event}
   > SetContentColorAndOpacity!(
    execute
    self:Border Object Reference=Get Border_EnabledState() -> (Border_EnabledState:Border Object Reference)
    InContentColorAndOpacity:Linear Color Structure=SelectColor(
     A=Get EnabledBorderColor() -> (EnabledBorderColor:Linear Color Structure)
     B=Get DisabledBorderColor() -> (DisabledBorderColor:Linear Color Structure)
     bPickA=@n4.bShowEnabled
    )
   ) -> (then)
   Event PreConstruct()
   > SetContentColorAndOpacity(self=Get Border_EnabledState(), InContentColorAndOpacity=Get EnabledBorderColor())
   Event OnCooldownStarted() {is=title K2Node_Event}
   > SetColorAndOpacity(
    self:User Widget Object Reference
    InColorAndOpacity:Linear Color Structure=MakeColor!(R:float=0.2, G:float=0.2, B:float=0.2, A:float=1) -> (ReturnValue:Linear Color Structure)
   )
   Event OnCooldownFinished() {is=title K2Node_Event}
   > SetColorAndOpacity(InColorAndOpacity=MakeColor(R=1, G=1, B=1, A=1))
   n5 = Event OnCooldownUpdated!() -> (OutputDelegate, then, CooldownDuration:float, CooldownRemaining:float, CooldownPercentage:float) {is=title K2Node_Event}
   n6 = Self-Reference()
   PrintString!(
    execute
    InString=Append!(A="COOLDOWN IS: ", B="To String (Float)"(InDouble=@n5.CooldownPercentage), C:String=" In: ", D:String=GetObjectName(Object=@n6)) -> (ReturnValue)
    bPrintToScreen
    bPrintToLog
    TextColor
    Duration
    Key
   ) -> (then)
   PrintString(InString=GetObjectName(Object=Self-Reference()))
  ]}
 IsKeyBound {nodes [
   IsKeyBound!() -> (then, Key:Key Structure=None) {is=title K2Node_FunctionEntry}
   > n1 = For Each Loop(Array:Array of User Widget Object References=GetAllEntries!(self:Dynamic Entry Box Base Object Reference=Get DynamicEntryBox_Keybinds() -> (DynamicEntryBox_Keybinds:Fort Dynamic Entry Box Object Reference)) -> (ReturnValue:Array of User Widget Object References)) -> (Array Element:User Widget Object Reference)
   Return Node!(execute=@n1.Completed, IsBound:bool=false)
   > n2 = Cast To KeybindWidget!(execute, Object=@n1) -> (then, CastFailed, AsKeybind Widget:Keybind Widget Object Reference) {is=title K2Node_DynamicCast}
   n3 = Get BoundKey!(self:Fort Keybind Widget Object Reference=@n2) -> (BoundKey:Key Structure)
   n4 = Get Key() -> (Key:Key Structure)
   > Branch(Condition=AND Boolean(
     A="Equal (Key)"(A:Key Structure=@n3, B:Key Structure=@n4) -> (ReturnValue:bool=~) {is=title K2Node_PromotableOperator}
     B=IsVisible(self=@n2)
    ))
   > Return Node(IsBound=true)
  ]}
 SetTextStyleSet {nodes [
   n1 = SetTextStyleSet!() -> (then, NewTextStyleSet:Data Table Object Reference) {is=title K2Node_FunctionEntry}
   > SetTextStyleSet!(
    execute
    self:Rich Text Block Object Reference=[
     Get RichText_Feedback() -> (RichText_Feedback:Common Rich Text Block Object Reference)
     Get RichText_Divider() -> (RichText_Divider:Common Rich Text Block Object Reference)
     Get FortRichText_AdditionalFeedback() -> (FortRichText_AdditionalFeedback:Common Rich Text Block Object Reference)
    ]
    NewTextStyleSet:Data Table Object Reference=@n1
   ) -> (then) {is=title K2Node_CallFunction}
  ]}
]
variables [
 DisabledBorderColor {type=LinearColor, default=<rgba 1 1 1 0.4>}
 EnabledBorderColor {type=LinearColor, default=<rgba 1 1 1 1>}
]
widgets {
 info {inheritedWidgetCount=5, parentClass=/Script/FortniteUI.AthenaHUDFeedbackLineCooldown, rootWidgetClass=&Overlay, widgetCount=12}
 widgets [
  Overlay_0 {widgetClassPath=&Overlay, widget=Overlay_0}
  Overlay_Glow {
   widgetClassPath=&Overlay
   bIsVariable=true
   parent=Overlay_0
   widget=Overlay_Glow
   properties {visibility=Collapsed}
   slot {class=OverlaySlot, properties {horizontalAlignment=HAlign_Center, verticalAlignment=VAlign_Center, padding=<ltrb -200 -200 -200 -200>}}
  }
  Image_Shade {
   widgetClassPath=&Image
   parent=Overlay_Glow
   widget=Image_Shade
   properties {
    visibility=SelfHitTestInvisible
    brush {
     bIsDynamicallyLoaded=false
     drawAs=Image
     imageType=NoImage
     mirroring=NoMirror
     resourceName=None
     resourceObject=/Game/Athena/HUD/Materials/M_CountdownBG_Alpha_Inst
     tiling=NoTile
     imageSize=<482.0438 58.3627>
     margin=<ltrb 0 0 0 0>
     outlineSettings {
      bUseBrushTransparency=false
      roundingType=HalfHeightRadius
      width=0
      color {colorUseRule=UseColor_Specified, specifiedColor=<rgba 0 0 0 0>}
      cornerRadii=<0 0 0 0>
     }
     tintColor {colorUseRule=UseColor_Specified, specifiedColor=<rgba 0.0156 0.0063 0.0063 1>}
     uVRegion {bIsValid=false, max=<0 0>, min=<0 0>}
    }
    colorAndOpacity=<rgba 1 1 1 0.2286>
   }
   slot {class=OverlaySlot, properties {horizontalAlignment=HAlign_Fill, verticalAlignment=VAlign_Center}}
  }
  Image_RedBG {
   widgetClassPath=&Image
   parent=Overlay_Glow
   widget=Image_RedBG
   properties {
    visibility=SelfHitTestInvisible
    brush {
     bIsDynamicallyLoaded=false
     drawAs=Image
     imageType=NoImage
     mirroring=NoMirror
     resourceName=None
     resourceObject=/Game/Athena/HUD/Materials/M_CountdownBG_Base
     tiling=NoTile
     imageSize=<514.2557 100>
     margin=<ltrb 0 0 0 0>
     outlineSettings {
      bUseBrushTransparency=false
      roundingType=HalfHeightRadius
      width=0
      color {colorUseRule=UseColor_Specified, specifiedColor=<rgba 0 0 0 0>}
      cornerRadii=<0 0 0 0>
     }
     tintColor {colorUseRule=UseColor_Specified, specifiedColor=<rgba 1 1 1 1>}
     uVRegion {bIsValid=false, max=<0 0>, min=<0 0>}
    }
    colorAndOpacity=<rgba 1 0 0 1>
   }
   slot {class=OverlaySlot, properties {horizontalAlignment=HAlign_Fill, verticalAlignment=VAlign_Center}}
  }
  Image_YellowBG {
   widgetClassPath=&Image
   parent=Overlay_Glow
   widget=Image_YellowBG
   properties {
    visibility=SelfHitTestInvisible
    brush {
     bIsDynamicallyLoaded=false
     drawAs=Image
     imageType=NoImage
     mirroring=NoMirror
     resourceName=None
     resourceObject=/Game/Athena/HUD/Materials/M_CountdownBG_Core
     tiling=NoTile
     imageSize=<514.2557 100>
     margin=<ltrb 0 0 0 0>
     outlineSettings {
      bUseBrushTransparency=false
      roundingType=HalfHeightRadius
      width=0
      color {colorUseRule=UseColor_Specified, specifiedColor=<rgba 0 0 0 0>}
      cornerRadii=<0 0 0 0>
     }
     tintColor {colorUseRule=UseColor_Specified, specifiedColor=<rgba 1 1 1 1>}
     uVRegion {bIsValid=false, max=<0 0>, min=<0 0>}
    }
    colorAndOpacity=<rgba 1 0.2789 0 1>
   }
   slot {class=OverlaySlot, properties {horizontalAlignment=HAlign_Fill}}
  }
  RichText_Divider {
   widgetClassPath=&CommonRichTextBlock
   bInherited=true
   bIsVariable=true
   parent=Overlay_0
   widget=RichText_Divider
   properties {
    bOverrideDefaultStyle=true
    text=|
    defaultTextStyleOverride {
     overflowPolicy=Clip
     transformPolicy=None
     colorAndOpacity {colorUseRule=UseColor_Specified, specifiedColor=<rgba 1 1 1 1>}
     font {
      bForceMonospaced=false
      bMaterialIsStencil=false
      fontMaterial=/BattleRoyaleGameUI/Shared/M_UI_Font_White
      fontObject=&BurbankSmall
      letterSpacing=0
      monospacedWidth=1
      size=16.5
      skewAmount=0
      typefaceFontName=Black
      outlineSettings {
       bApplyOutlineToDropShadows=true
       bMiteredCorners=false
       bSeparateFillAlpha=true
       outlineMaterial=/BattleRoyaleGameUI/Shared/MI_UI_Font_SoftShadow_SmallText
       outlineSize=4
       outlineColor=<rgba 1 1 1 1>
      }
     }
     highlightColor {colorUseRule=UseColor_Specified, specifiedColor=<rgba 0 0 0 1>}
     highlightShape {
      bIsDynamicallyLoaded=false
      drawAs=Image
      imageType=NoImage
      mirroring=NoMirror
      resourceName=None
      resourceObject=null
      tiling=NoTile
      imageSize=<32 32>
      margin=<ltrb 0 0 0 0>
      outlineSettings {
       bUseBrushTransparency=false
       roundingType=HalfHeightRadius
       width=0
       color {colorUseRule=UseColor_Specified, specifiedColor=<rgba 0 0 0 0>}
       cornerRadii=<0 0 0 0>
      }
      tintColor {colorUseRule=UseColor_Specified, specifiedColor=<rgba 1 1 1 1>}
      uVRegion {bIsValid=false, max=<0 0>, min=<0 0>}
     }
     selectedBackgroundColor {colorUseRule=UseColor_Foreground, specifiedColor=<rgba 1 0 1 0>}
     shadowColorAndOpacity=<rgba 0 0 0 1>
     shadowOffset=<0 0>
     strikeBrush {
      bIsDynamicallyLoaded=false
      drawAs=Image
      imageType=NoImage
      mirroring=NoMirror
      resourceName=None
      resourceObject=null
      tiling=NoTile
      imageSize=<32 32>
      margin=<ltrb 0 0 0 0>
      outlineSettings {
       bUseBrushTransparency=false
       roundingType=HalfHeightRadius
       width=0
       color {colorUseRule=UseColor_Specified, specifiedColor=<rgba 0 0 0 0>}
       cornerRadii=<0 0 0 0>
      }
      tintColor {colorUseRule=UseColor_Specified, specifiedColor=<rgba 1 1 1 1>}
      uVRegion {bIsValid=false, max=<0 0>, min=<0 0>}
     }
     underlineBrush {
      bIsDynamicallyLoaded=false
      drawAs=Image
      imageType=NoImage
      mirroring=NoMirror
      resourceName=None
      resourceObject=null
      tiling=NoTile
      imageSize=<32 32>
      margin=<ltrb 0 0 0 0>
      outlineSettings {
       bUseBrushTransparency=false
       roundingType=HalfHeightRadius
       width=0
       color {colorUseRule=UseColor_Specified, specifiedColor=<rgba 0 0 0 0>}
       cornerRadii=<0 0 0 0>
      }
      tintColor {colorUseRule=UseColor_Specified, specifiedColor=<rgba 1 1 1 1>}
      uVRegion {bIsValid=false, max=<0 0>, min=<0 0>}
     }
    }
   }
   slot {class=OverlaySlot, properties {horizontalAlignment=HAlign_Fill, verticalAlignment=VAlign_Bottom}}
  }
  Border_EnabledState {
   widgetClassPath=&Border
   bIsVariable=true
   parent=Overlay_0
   widget=Border_EnabledState
   properties {
    background {
     bIsDynamicallyLoaded=false
     drawAs=NoDrawType
     imageType=NoImage
     mirroring=NoMirror
     resourceName=None
     resourceObject=null
     tiling=NoTile
     imageSize=<32 32>
     margin=<ltrb 0 0 0 0>
     outlineSettings {
      bUseBrushTransparency=false
      roundingType=HalfHeightRadius
      width=0
      color {colorUseRule=UseColor_Specified, specifiedColor=<rgba 0 0 0 0>}
      cornerRadii=<0 0 0 0>
     }
     tintColor {colorUseRule=UseColor_Specified, specifiedColor=<rgba 1 1 1 1>}
     uVRegion {bIsValid=false, max=<0 0>, min=<0 0>}
    }
    padding=<ltrb 0 0 0 0>
   }
   slot {class=OverlaySlot, properties {padding=<ltrb 15 0 0 0>}}
  }
  FeedbackLine {
   widgetClassPath=&HorizontalBox
   parent=Border_EnabledState
   widget=FeedbackLine
   slot {class=BorderSlot, properties {padding=<ltrb 0 0 0 0>}}
  }
  DynamicEntryBox_Keybinds {
   widgetClassPath=/Script/FortniteUI.FortDynamicEntryBox
   bInherited=true
   bIsVariable=true
   parent=FeedbackLine
   widget=DynamicEntryBox_Keybinds
   properties {
    entryHorizontalAlignment=HAlign_Left
    entryWidgetClass=&/WBP_BRGameUI_ContextualControls_KeybindWidget.WBP_BRGameUI_ContextualControls_KeybindWidget_C
    numDesignerPreviewEntries=2
    entrySpacing=<5 3>
   }
   slot {class=HorizontalBoxSlot}
  }
  RichText_Feedback {
   widgetClassPath=&CommonRichTextBlock
   bInherited=true
   parent=FeedbackLine
   widget=RichText_Feedback
   properties {
    bOverrideDefaultStyle=true
    text="{ContextualAction}"
    textTransformPolicy=ToUpper
    defaultTextStyleOverride {
     overflowPolicy=Clip
     transformPolicy=None
     colorAndOpacity {colorUseRule=UseColor_Specified, specifiedColor=<rgba 1 1 1 1>}
     font {
      bForceMonospaced=false
      bMaterialIsStencil=false
      fontMaterial=&M_UI_Font_White
      fontObject=&BurbankSmall
      letterSpacing=0
      monospacedWidth=1
      size=16.5
      skewAmount=0
      typefaceFontName=Black
      outlineSettings {
       bApplyOutlineToDropShadows=true
       bMiteredCorners=false
       bSeparateFillAlpha=true
       outlineMaterial=&MI_UI_Font_SoftShadow_SmallText
       outlineSize=4
       outlineColor=<rgba 1 1 1 1>
      }
     }
     highlightColor {colorUseRule=UseColor_Specified, specifiedColor=<rgba 0 0 0 1>}
     highlightShape {
      bIsDynamicallyLoaded=false
      drawAs=Image
      imageType=NoImage
      mirroring=NoMirror
      resourceName=None
      resourceObject=null
      tiling=NoTile
      imageSize=<32 32>
      margin=<ltrb 0 0 0 0>
      outlineSettings {
       bUseBrushTransparency=false
       roundingType=HalfHeightRadius
       width=0
       color {colorUseRule=UseColor_Specified, specifiedColor=<rgba 0 0 0 0>}
       cornerRadii=<0 0 0 0>
      }
      tintColor {colorUseRule=UseColor_Specified, specifiedColor=<rgba 1 1 1 1>}
      uVRegion {bIsValid=false, max=<0 0>, min=<0 0>}
     }
     selectedBackgroundColor {colorUseRule=UseColor_Foreground, specifiedColor=<rgba 1 0 1 0>}
     shadowColorAndOpacity=<rgba 0 0 0 1>
     shadowOffset=<0 0>
     strikeBrush {
      bIsDynamicallyLoaded=false
      drawAs=Image
      imageType=NoImage
      mirroring=NoMirror
      resourceName=None
      resourceObject=null
      tiling=NoTile
      imageSize=<32 32>
      margin=<ltrb 0 0 0 0>
      outlineSettings {
       bUseBrushTransparency=false
       roundingType=HalfHeightRadius
       width=0
       color {colorUseRule=UseColor_Specified, specifiedColor=<rgba 0 0 0 0>}
       cornerRadii=<0 0 0 0>
      }
      tintColor {colorUseRule=UseColor_Specified, specifiedColor=<rgba 1 1 1 1>}
      uVRegion {bIsValid=false, max=<0 0>, min=<0 0>}
     }
     underlineBrush {
      bIsDynamicallyLoaded=false
      drawAs=Image
      imageType=NoImage
      mirroring=NoMirror
      resourceName=None
      resourceObject=null
      tiling=NoTile
      imageSize=<32 32>
      margin=<ltrb 0 0 0 0>
      outlineSettings {
       bUseBrushTransparency=false
       roundingType=HalfHeightRadius
       width=0
       color {colorUseRule=UseColor_Specified, specifiedColor=<rgba 0 0 0 0>}
       cornerRadii=<0 0 0 0>
      }
      tintColor {colorUseRule=UseColor_Specified, specifiedColor=<rgba 1 1 1 1>}
      uVRegion {bIsValid=false, max=<0 0>, min=<0 0>}
     }
    }
    margin=<ltrb -4 0 -4 0>
   }
   slot {class=HorizontalBoxSlot, properties {verticalAlignment=VAlign_Center, padding=<ltrb 10 0 0 0>}}
  }
  FortRichText_AdditionalFeedback {
   widgetClassPath=&CommonRichTextBlock
   bInherited=true
   parent=FeedbackLine
   widget=FortRichText_AdditionalFeedback
   properties {
    bOverrideDefaultStyle=true
    text="<img id=\"WoodIcon\" width=\"42\"></><img id=\"StoneIcon\" width=\"42\"></><img id=\"MetalIcon\" width=\"42\"></>"
    decoratorClasses [/Game/Athena/HUD/AthenaHUDFeedbackLineRichTextDecorators.AthenaHUDFeedbackLineRichTextDecorators_C]
    defaultTextStyleOverride {
     overflowPolicy=Clip
     transformPolicy=None
     colorAndOpacity {colorUseRule=UseColor_Specified, specifiedColor=<rgba 1 1 1 1>}
     font {
      bForceMonospaced=false
      bMaterialIsStencil=false
      fontMaterial=/BattleRoyaleGameUI/Shared/MI_UI_Font_Cyan
      fontObject=&BurbankSmall
      letterSpacing=0
      monospacedWidth=1
      size=15
      skewAmount=0
      typefaceFontName=Black
      outlineSettings {
       bApplyOutlineToDropShadows=true
       bMiteredCorners=false
       bSeparateFillAlpha=true
       outlineMaterial=&MI_UI_Font_SoftShadow_SmallText
       outlineSize=4
       outlineColor=<rgba 1 1 1 1>
      }
     }
     highlightColor {colorUseRule=UseColor_Specified, specifiedColor=<rgba 0 0 0 1>}
     highlightShape {
      bIsDynamicallyLoaded=false
      drawAs=Image
      imageType=NoImage
      mirroring=NoMirror
      resourceName=None
      resourceObject=null
      tiling=NoTile
      imageSize=<32 32>
      margin=<ltrb 0 0 0 0>
      outlineSettings {
       bUseBrushTransparency=false
       roundingType=HalfHeightRadius
       width=0
       color {colorUseRule=UseColor_Specified, specifiedColor=<rgba 0 0 0 0>}
       cornerRadii=<0 0 0 0>
      }
      tintColor {colorUseRule=UseColor_Specified, specifiedColor=<rgba 1 1 1 1>}
      uVRegion {bIsValid=false, max=<0 0>, min=<0 0>}
     }
     selectedBackgroundColor {colorUseRule=UseColor_Foreground, specifiedColor=<rgba 1 0 1 0>}
     shadowColorAndOpacity=<rgba 0 0 0 1>
     shadowOffset=<0 0>
     strikeBrush {
      bIsDynamicallyLoaded=false
      drawAs=Image
      imageType=NoImage
      mirroring=NoMirror
      resourceName=None
      resourceObject=null
      tiling=NoTile
      imageSize=<32 32>
      margin=<ltrb 0 0 0 0>
      outlineSettings {
       bUseBrushTransparency=false
       roundingType=HalfHeightRadius
       width=0
       color {colorUseRule=UseColor_Specified, specifiedColor=<rgba 0 0 0 0>}
       cornerRadii=<0 0 0 0>
      }
      tintColor {colorUseRule=UseColor_Specified, specifiedColor=<rgba 1 1 1 1>}
      uVRegion {bIsValid=false, max=<0 0>, min=<0 0>}
     }
     underlineBrush {
      bIsDynamicallyLoaded=false
      drawAs=Image
      imageType=NoImage
      mirroring=NoMirror
      resourceName=None
      resourceObject=null
      tiling=NoTile
      imageSize=<32 32>
      margin=<ltrb 0 0 0 0>
      outlineSettings {
       bUseBrushTransparency=false
       roundingType=HalfHeightRadius
       width=0
       color {colorUseRule=UseColor_Specified, specifiedColor=<rgba 0 0 0 0>}
       cornerRadii=<0 0 0 0>
      }
      tintColor {colorUseRule=UseColor_Specified, specifiedColor=<rgba 1 1 1 1>}
      uVRegion {bIsValid=false, max=<0 0>, min=<0 0>}
     }
    }
   }
   slot {class=HorizontalBoxSlot, properties {horizontalAlignment=HAlign_Left, verticalAlignment=VAlign_Center, padding=<ltrb 10 0 0 0>}}
  }
  TutorialHighlight {widgetClassPath=&TutorialHighlight}
 ]
}
